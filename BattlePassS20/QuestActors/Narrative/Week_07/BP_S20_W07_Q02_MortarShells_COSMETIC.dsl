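; Cosmetic mortar-shells actor for a Season 20 Week 7 narrative quest, which plays a one-shot shell swap effect.
; On BeginPlay it caches the mesh's initial relative position and rotation, attaches a swap sound and starts the ShellsSwapped timeline.
; The timeline drives the mesh's relative height, rotation and scale, without reading the cached values, then destroys the actor when finished.
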(event EventBeginPlay
  (bind _sm_mortarshells (Variables|Default|GetSM_MortarShells))
  (Variables|Default|SetInitialPosition (Variables|/Script/Engine.SceneComponent|GetRelativeLocation _sm_mortarshells))
  (Variables|Default|SetInitialRotation (Variables|/Script/Engine.SceneComponent|GetRelativeRotation _sm_mortarshells))
  (Audio|SpawnSoundAttached "/BattlePassS20/QuestActors/Narrative/Week_07/SFX/S20_DataMunition_Swap_Cue.S20_DataMunition_Swap_Cue" (Variables|Default|GetSM_MortarShells))
  (bind (_direction _mortarshellsheight _mortarshellsrotation _mortarshellsscale) (Timeline|ShellsSwapped PlayFromStart)
    (:Update
      (Transformation|SetRelativeTransform (Variables|Default|GetSM_MortarShells) (Math|Transform|MakeTransform (Math|Vector|MakeVector 0.0 0.0 _mortarshellsheight) (Math|Rotator|MakeRotator (.x _mortarshellsrotation) (.y _mortarshellsrotation) (.z _mortarshellsrotation)) _mortarshellsscale) false false))
    (:Finished
      (Actor|DestroyActor))))

(fn ConstructionScript)
