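; Cosmetic server actor for a Season 20 Week 7 narrative quest, playing a one-shot hacked effect for the server mesh and then removing itself.
; On BeginPlay it stores the mesh's relative location and rotation, spawns an attached data-acquired sound and starts the ServerHacked timeline.
; The timeline drives relative Z height and scale with zero rotation, without using the stored values, and destroys the actor when finished.
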(event EventBeginPlay
  (bind _sm_server (Variables|Default|GetSM_Server))
  (Variables|Default|SetInitialPosition (Variables|/Script/Engine.SceneComponent|GetRelativeLocation _sm_server))
  (Variables|Default|SetInitialRotation (Variables|/Script/Engine.SceneComponent|GetRelativeRotation _sm_server))
  (Audio|SpawnSoundAttached "/BattlePassS20/QuestActors/Narrative/Week_07/SFX/ServerData_Acquired_Cue.ServerData_Acquired_Cue" (Variables|Default|GetSM_Server))
  (bind (_direction _serverheight _serverscale) (Timeline|ServerHacked PlayFromStart)
    (:Update
      (Transformation|SetRelativeTransform (Variables|Default|GetSM_Server) (Math|Transform|MakeTransform (Math|Vector|MakeVector 0.0 0.0 _serverheight) "0, 0, 0" _serverscale) false false))
    (:Finished
      (Actor|DestroyActor))))

(fn ConstructionScript)
